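; Butterfly event actor that flies a butterfly in to land on each client's local or spectated player pawn.
; The ButterflyApproach timeline drives spline position, material-based wing flapping and fade, while the actor trails the pawn's attach socket.
; Clients feed the butterfly's location to the player anim blueprint and modulate audio by movement speed.
; The server triggers the landing montage, replicates input lock and loop effects, then destroys the actor.
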(event Custom|StartupSpawn
  (Networking|SwitchHasAuthority
    (:Remote
      (Components|Activation|SetActive (Variables|Default|GetP_Butterfly_Spawn) true false)
      (bind _returnvalue (Audio|SpawnSound2D (Variables|Default|GetSoundOnSpawn)))
      (Variables|Default|SetButterflySequenceAudio _returnvalue))))

(event EventBeginPlay
  (bind _returnvalue (Fortnite|GetLocalFortPlayerControllers true))
  (bind _output (Utilities|Array|Get(acopy) _returnvalue))
  (bind _pawn (Variables|Default|GetPawn))
  (Networking|SwitchHasAuthority
    (:Authority
      (CallFunction|ServerStartTimeline)
      (CallFunction|StartupSpawn))
    (:Remote
      (Variables|Default|SetStartRotation (Transformation|GetActorRotation))
      (if (Replay|GetIsPlayingReplay)
        (bind _output_get (Variables|Default|SetLocalController (Utilities|Array|Get(acopy) (Fortnite|GetLocalFortPlayerControllers false true))))
        (bind _asfort_replay_spectator_athena (Utilities|Casting|CastToFortReplaySpectatorAthena _output_get))
        (Utilities|FlowControl|Delay 0.25)
        (Utilities|IsValid (PlayerState|GetCurrentPawn (Class|FortPlayerControllerSpectating|GetFollowedPlayer _asfort_replay_spectator_athena))
          (:Is%20Valid
            _)
          (:Is%20Not%20Valid
            _))
        (Variables|Default|SetPawn (PlayerState|GetCurrentPawn (Class|FortPlayerControllerSpectating|GetFollowedPlayer _asfort_replay_spectator_athena)))
        (else
          (Utilities|IsValid _output
            (:Is%20Valid
              (bind _output_get_2 (Variables|Default|SetLocalController _output))
              (Utilities|IsValid (Game|Player|GetPlayerPawn _output_get_2)
                (:Is%20Valid
                  (Variables|Default|SetPawn (Game|Player|GetPlayerPawn _output_get_2)))
                (:Is%20Not%20Valid
                  ;[1 1 1] If dead and spectating
                  (bind _asfort_player_pawn_athena (Utilities|Casting|CastToFortPlayerPawnAthena (Pawn|GetViewTarget (Variables|Default|GetLocalController)))
                    (:then
                      (Variables|Default|SetPawn _asfort_player_pawn_athena))
                    (:CastFailed
                      (Actor|DestroyActor))))))
            (:Is%20Not%20Valid
              (Actor|DestroyActor)))))
      (bind _mesh (Variables|/Script/Engine.Character|GetMesh _pawn)
        (:then
          (Variables|Default|SetPlayerMesh _mesh)
          (bind _output_get_3 (Variables|Default|SetLocation (Transformation|GetActorLocation _pawn)))
          (Transformation|SetActorLocation _output_get_3 false true))
        (:else
          (Actor|DestroyActor))))))

(event Custom|serverStartTimeline
  (bind (_returnvalue_roll _returnvalue_pitch _returnvalue_yaw) (Math|Rotator|BreakRotator (Transformation|GetActorRotation (Variables|Default|GetPawn))))
  (bind (_returnvalue_location _returnvalue_rotation) (Math|Transform|BreakTransform (Spline|GetTransformatTime (Variables|Default|GetSpline) (Variables|Default|GetSplineTime) "Local" false false)))
  (bind _sm_butterfly (Variables|Default|GetSM_Butterfly))
  (bind _sm_butterfly_2 (Variables|Default|GetSM_Butterfly))
  (Utilities|FlowControl|Sequence
    (:then_0
      (bind (_direction _fadein _location _flaptime _wingbend _followinterp _syncrotation) (Timeline|ButterflyApproach Play)
        (:Update
          (Networking|SwitchHasAuthority
            (:Remote
              (Variables|Default|SetSplineTime _location)
              (Utilities|FlowControl|Sequence
                (:then_0
                  ;[1 1 1] Make player character look at butterfly
                  (bind _playermesh (Variables|Default|GetPlayerMesh)
                    (:then
                      (bind _asfortnite_butterfly_ghost_anim_blueprint (Utilities|Casting|CastToFortnite_Butterfly_Ghost_AnimBlueprint (Components|SkeletalMesh|GetAnimInstance _playermesh)))
                      (Variables|/Game/Athena/Prototype/Blueprints/Island/Fortnite_Butterfly_Ghost_AnimBlueprint.Fortnite_Butterfly_Ghost_AnimBlueprint_C|SetButterFlyLocation (Transformation|GetWorldLocation (Variables|Default|GetSM_Butterfly)) _asfortnite_butterfly_ghost_anim_blueprint))
                    (:else)))
                (:then_1
                  ; Butterfly animation params
                  (Transformation|SetRelativeScale3D _sm_butterfly (Math|Conversions|ToVector(Float) (Math|Float|Lerp 4.0 0.075 _fadein)))
                  (Transformation|SetRelativeLocationAndRotation _sm_butterfly _returnvalue_location (Math|Rotator|MakeRotator (.roll _returnvalue_rotation) (+ (* _fadein 37.0) (.pitch _returnvalue_rotation)) (.yaw _returnvalue_rotation)) false false)
                  (Rendering|Material|SetScalarParameterValueonMaterials _sm_butterfly "Opacity" _fadein)
                  (Rendering|Material|SetScalarParameterValueonMaterials _sm_butterfly "FlapTime" (* _flaptime 3e+01))
                  (Rendering|Material|SetScalarParameterValueonMaterials _sm_butterfly "WingBend" _wingbend)
                  (Rendering|Material|SetVectorParameterValueonMaterials _sm_butterfly_2 "Axis" (Transformation|GetUpVector _sm_butterfly_2))
                  (Rendering|Material|SetVectorParameterValueonMaterials _sm_butterfly_2 "Pivot" (Transformation|GetWorldLocation _sm_butterfly_2)))
                (:then_2
                  ; Stick to player, move their camera
                  (Variables|Default|SetPreviousLocation (Variables|Default|GetLocation))
                  (bind _output_get (Variables|Default|SetLocation (Math|Interpolation|VInterpTo (Variables|Default|GetLocation) (Transformation|GetSocketLocation (Variables|Default|GetPlayerMesh) (Variables|Default|GetButterflyAttachPoint)) (Utilities|Time|GetWorldDeltaSeconds) _followinterp)))
                  (Transformation|SetActorLocation _output_get)
                  (Transformation|SetActorRotation (Math|Rotator|Lerp(Rotator) (Variables|Default|GetStartRotation) (Math|Rotator|MakeRotator _returnvalue_roll _returnvalue_pitch (+ _returnvalue_yaw 1.8e+02)) _syncrotation true))
                  (CallFunction|EvaluateVelocity))
                (:then_3
                  (bind _butterflysequenceaudio (Variables|Default|GetButterflySequenceAudio)
                    (:then
                      (if (not (Audio|Components|Audio|IsPlaying _butterflysequenceaudio))
                        (Audio|Components|Audio|Play _butterflysequenceaudio 0.0)))
                    (:else)))))))
        (:Finished
          (Networking|SwitchHasAuthority
            (:Authority
              (Variables|Default|SetLockInput false)
              (Utilities|FlowControl|Delay 3.0)
              (Actor|DestroyActor self))
            (:Remote
              (Audio|PlaySound2D "/Game/Athena/Sounds/Events/Butterfly/Butterfly_Plop_End_Cue.Butterfly_Plop_End_Cue")
              (Rendering|SetActorHiddenInGame true))))
        (:BeginLanding
          (Networking|SwitchHasAuthority
            (:Authority
              (Variables|Default|SetLockInput true)
              (CallFunction|PlayMontage))))
        (:Landing
          (Networking|SwitchHasAuthority
            (:Remote
              (Effects|SpawnEmitterAttached :EmitterTemplate "/Game/Athena/Prototype/Blueprints/Island/P_Butterfly_Contact.P_Butterfly_Contact" :AttachToComponent (Variables|/Script/Engine.Character|GetMesh (Variables|Default|GetPawn)) :AttachPointName (Variables|Default|GetButterflyAttachPoint) :Scale "1.000000,1.000000,1.000000" :LocationType "SnapToTarget" :bAutoDestroy true :PoolingMethod "None" :bAutoActivate true))))))
    (:then_1
      (Utilities|FlowControl|Delay 5.0)
      (Variables|Default|SetLoopEffectsVisible true))))

(event Custom|PlayMontage
  (Networking|SwitchHasAuthority
    (:Remote
      (Audio|PlaySound2D (Variables|Default|GetSoundOnCameraSwitch))
      (Animation|PlayAnimMontage (Variables|Default|GetPawn) "/Game/Animation/Game/MainPlayer/Combat/Gadgets/Medium/Male/GhostRock/GhostRock_ButterflySequence_CMM_M.GhostRock_ButterflySequence_CMM_M" 1.0 "None"))))

(event Custom|Evaluate%20Velocity
  (bind _movementaudio (Variables|Default|GetMovementAudio))
  (bind _output_get (Variables|Default|SetCurrentLocDelta (Math|Interpolation|FInterpTo (Variables|Default|GetCurrentLocDelta) (Math|Float|MapRangeClamped (Math|Vector|VectorLength (- (Variables|Default|GetLocation) (Variables|Default|GetPreviousLocation))) 0.0 (Variables|Default|GetMaxLocDelta) 0.0 1.0) (Utilities|Time|GetWorldDeltaSeconds) 1.0)))
  (Audio|Components|Audio|SetVolumeMultiplier _movementaudio (Math|Float|MapRangeClamped _output_get 0.0 1.0 0.5 1.0) :!class /Script/Engine.AudioComponent)
  (Audio|Parameter|SetFloatParameter _movementaudio "SplineTime" (- 1.0 (Variables|Default|GetSplineTime)) :!class /Script/AudioExtensions.AudioParameterControllerInterface)
  (Audio|Components|Audio|SetPitchMultiplier (Variables|Default|GetMovementAudio) (Math|Float|MapRangeClamped _output_get 0.0 1.0 0.8 1.2)))

(fn OnRep_LockInput
  (bind _lockinput (Variables|Default|GetLockInput))
  (bind _localcontroller (Variables|Default|GetLocalController))
  (Networking|SwitchHasAuthority
    (:Remote
      (if (not (Replay|GetIsPlayingReplay))
        (Input|SetIgnoreLookInput _localcontroller _lockinput)
        (Input|SetIgnoreMoveInput _localcontroller _lockinput)))))

(fn OnRep_LoopEffectsVisible
  (Components|Activation|SetActive (Variables|Default|GetP_Butterfly_Loop) (Variables|Default|GetLoopEffectsVisible) false))

(fn ConstructionScript
  (Rendering|Material|SetVectorParameterValueonMaterials (Variables|Default|GetSM_Butterfly) "Pivot" (Transformation|GetWorldLocation (Variables|Default|GetSM_Butterfly))))
