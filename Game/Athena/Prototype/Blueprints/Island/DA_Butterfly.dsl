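; DA_Butterfly is a prototype island deco tool Blueprint that turns a primary-fire press into a gameplay ability event.
; On trigger press, it sends a gameplay event tagged Abilities.Generic.ActiveAbility.Activate to its instigator, with the same tag in the event data.
; Secondary-fire press and release handlers and the construction script are empty, and unequip only calls the parent.
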(event DecoTool|EventOnPressSecondaryFire
  (params (FortDecoHelper :type Fort%20Deco%20Helper%20Object%20Reference)))

(event DecoTool|EventOnReleaseSecondaryFire
  (params (FortDecoHelper :type Fort%20Deco%20Helper%20Object%20Reference)))

(event DecoTool|EventOnPressTrigger
  (params (FortDecoHelper :type Fort%20Deco%20Helper%20Object%20Reference))
  ;[0.3904 0.4224 1] When you press the primary fire send this event
  (Ability|SendGameplayEventtoActor (Game|GetInstigator) "(TagName=\"Abilities.Generic.ActiveAbility.Activate\")" (Utilities|Struct|MakeGameplayEventData "(TagName=\"Abilities.Generic.ActiveAbility.Activate\")")))

(event DecoTool|EventOnUnEquip
  (|CallParentFunction))

(fn ConstructionScript)
