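# Animation Blueprint for the ghost in the island butterfly prototype, targeting the average male Fortnite player skeleton.
# The AnimGraph holds only an Output Pose node with nothing feeding it, so it defines no animation logic.
# The EventGraph is empty.
# A ButterFlyLocation vector variable is declared, but neither graph reads or writes it.
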AnimBlueprint /Game/Athena/Prototype/Blueprints/Island/Fortnite_Butterfly_Ghost_AnimBlueprint
target_skeleton=&Fortnite_M_Avg_Player_Skeleton
graphs [
 AnimGraph {nodes [Output Pose() {layerGroup=None}]}
 EventGraph {links [], nodes []}
]
variables [ButterFlyLocation {type=Vector}]
