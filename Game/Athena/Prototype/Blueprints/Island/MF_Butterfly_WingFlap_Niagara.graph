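# Material function that outputs a position offset animating butterfly wing flaps on Niagara mesh particles.
# It rotates vertices around the local Z axis at the particle position, combining a travelling sine along particle Z, a constant offset and a time-driven flap, mirrored by the sign of particle-space Y.
# A local-space displacement from UV channel 3, scaled by Offset and Spread, is added.
# Two dynamic parameter sets supply animation time and shape controls.
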MaterialFunction /Game/Athena/Prototype/Blueprints/Island/MF_Butterfly_WingFlap_Niagara
graph {nodes [
  n1 = TextureCoordinate() {coordinateIndex=3}
  n2 = AppendVector(A=Constant(), B=ComponentMask(None=@n1) {r=true})
  n3 = OneMinus(None=ComponentMask(None=@n1) {g=true})
  n4 = Transform(None=Add(A=AppendVector(A=@n2, B=@n3), B=Constant3Vector() {constant=<rgba 0 -0.5 -0.5 0>}) {constB=-0.5}) {transformSourceType=TRANSFORMSOURCE_Local}
  n5 = DynamicParameter() {parameterIndex=1, paramNames [Wave Length, Param2, Spread, FlapMult]}
  n6 = Multiply(A=ScalarParameter() {defaultValue=2000, parameterName=Offset}, B=@n5.Spread)
  n7 = Normalize(VectorInput=Transform(None=Constant3Vector() {constant=<rgba 0 0 1 0>}) {transformSourceType=TRANSFORMSOURCE_Local})
  n8 = Transform(None=Subtract(A=WorldPosition() {worldPositionShaderOffset=WPT_ExcludeAllShaderOffsets}.XYZ, B=ParticlePositionWS())) {transformSourceType=TRANSFORMSOURCE_World, transformType=TRANSFORM_ParticleWorld}
  n9 = Divide(
   A=MaterialFunctionCall(Float3=@n8) {materialFunction=&BreakOutFloat3Components}.B
   B=@n5.Wave Length
  ) {constB=500}
  n10 = DynamicParameter() {paramNames [Wave Amplitude, Wave Offset, Flap Amplitude, Time]}
  n11 = Multiply(A=Sine(None=Add(A=@n9, B=@n10.Time)), B=@n10.Wave Amplitude)
  n12 = Sine(None=@n10.Time)
  n13 = Multiply(
   A=Add(A=Add(A=@n11, B=@n10.Wave Offset), B=Multiply(A=@n12, B=@n10.Flap Amplitude))
   B=Sign(None=MaterialFunctionCall(Float3=@n8) {materialFunction=&BreakOutFloat3Components}.G)
  )
  n14 = ParticlePositionWS()
  n15 = WorldPosition() {worldPositionShaderOffset=WPT_ExcludeAllShaderOffsets}
  FunctionOutput(None=Add(A=Multiply(A=@n4, B=@n6), B=RotateAboutAxis(NormalizedRotationAxis=@n7, RotationAngle=@n13, PivotPoint=@n14, Position=@n15.XYZ)))
  LinearInterpolate(Alpha=Saturate(None=LinearInterpolate(Alpha=@n5.Spread) {constA=1, constB=-10}))
 ]}
