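# Translucent, unlit, two-sided material for Niagara mesh particles that renders glowing, cracked-glass butterfly shards.
# Emissive adds tinted Fresnel cubemap reflections to layered crack masks over a UV gradient, scaled by particle color.
# Opacity combines particle alpha, gradient and parameter, with a switchable camera-distance fade taking one dynamic parameter for both start and end.
# Layered crack normals optionally blend with a Niagara wing-flap function's normal through a static switch.
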Material /Game/Athena/Prototype/Blueprints/Island/M_ButterFly_Shard_Niagara
graph {
 properties {
  bUsedWithNiagaraMeshParticles=true
  blendMode=BLEND_Translucent
  pixelDepthOffsetMode=PDOM_Legacy
  refractionMethod=RM_IndexOfRefraction
  shadingModel=MSM_Unlit
  twoSided=true
  usedShadingModels=Unlit
 }
 nodes [
  Constant() {r=1}
  n1 = ParticleColor()
  n2 = ReflectionVectorWS() {bNormalizeCustomWorldNormal=true}
  n3 = Multiply(
   A=Multiply(
    A=TextureSampleParameterCube!(UVs=@n2, Apply View MipBias) {parameterName=Cubemap, samplerType=SAMPLERTYPE_LinearColor, texture=/Game/Packages/Fortress_Sky/TexturesHDR/T_AthenaSkylight}.RGB
    B=VectorParameter() {parameterName=CubemapTint, defaultValue=<rgba 0.4196 0.7965 1 1>}.RGB
   ) {constB=0.7}
   B=Fresnel() {baseReflectFraction=0.3, exponent=3}
  )
  n4 = ScalarParameter() {defaultValue=1, parameterName=Opacity}
  n5 = TextureCoordinate() {coordinateIndex=1}
  n6 = Add(
   A=LinearInterpolate(A=Constant2Vector() {g=0.25}, Alpha=@n4) {constA=0.5, constB=0}
   B=Add(A=TextureCoordinate() {coordinateIndex=2, uTiling=1.25, vTiling=0.5}, B=Multiply(A=@n5) {constB=0.5})
  )
  n7 = Add(
   A=Multiply(A=TextureSample(UVs=@n6) {texture=/Game/Effects/Fort_Effects/Textures/Tile/T_Glass_Cracks_M}.G) {constB=0.6}
   B=TextureSample(UVs=TextureCoordinate()) {texture=&T_Glass_Cracks_M}.G
  )
  n8 = VectorParameter() {parameterName=CrackColor, defaultValue=<rgba 0.2058 0.778 1 0>}
  n9 = Multiply(A=@n8.RGB) {constB=8}
  n10 = VectorParameter() {parameterName=BaseColor, defaultValue=<rgba 0.1405 0.1916 0.5 0>}
  n11 = Power(Base=ComponentMask(None=TextureCoordinate() {coordinateIndex=2}) {g=true})
  n12 = Multiply(
   A=LinearInterpolate(Alpha=Fresnel() {baseReflectFraction=0, exponent=3}) {constA=1, constB=2}
   B=Add(A=Multiply(A=@n7, B=@n9), B=LinearInterpolate(A=@n10.RGB, B=@n8.RGB, Alpha=@n11))
  )
  MaterialFunctionCall(Emissive=Multiply(A=@n1.RGB, B=Add(A=@n3, B=@n12))) {materialFunction=&MF_FX_EmissiveAdjustments}
  n13 = Multiply(A=@n4)
  n14 = Multiply(A=@n1.A, B=Multiply(A=LinearInterpolate(Alpha=@n11) {constA=0.2, constB=1.05}, B=LinearInterpolate(A=@n4, Alpha=@n13)) {constB=1.5})
  n15 = DynamicParameter() {parameterIndex=2, defaultValue=<rgba 500 300 1 1>, paramNames [Fade_Start_Fade_Here, Fade_End_Fade_Here, Param3, Param4]}
  n16 = MaterialFunctionCall(Fade_Start_Distance=@n15.Fade_Start_Fade_Here, Fade_End_Distance=@n15.Fade_Start_Fade_Here) {materialFunction=&MF_Distance_SyncedFade_FX}
  StaticSwitchParameter(True=Multiply(A=@n14, B=@n16.Result), False=@n14) {parameterName=UseFadeWhenCloseToCam?}
  n17 = Add(A=TextureCoordinate() {coordinateIndex=2, uTiling=1.25, vTiling=0.5}, B=Multiply(A=TextureCoordinate() {coordinateIndex=1}) {constB=0.5})
  n18 = Add(
   A=Multiply(
    A=TextureSampleParameter2D() {parameterName=Normals, samplerType=SAMPLERTYPE_Normal, texture=/Game/Effects/Fort_Effects/Textures/Tile/T_Glass_Cracks_N}.RGB
    B=Constant3Vector() {constant=<rgba 2 2 1 0>}
   )
   B=Multiply(
    A=TextureSampleParameter2D(UVs=@n17) {parameterName=NormalsTile, samplerType=SAMPLERTYPE_Normal, texture=&T_Glass_Cracks_N}.RGB
    B=Constant3Vector() {constant=<rgba 1 1 0 0>}
   )
  )
  StaticSwitchParameter(
   True=MaterialFunctionCall(
    BaseNormal=MaterialFunctionCall() {materialFunction=&/MF_Butterfly_WingFlap_Niagara}.Result
    AdditionalNormal=@n18
   ) {materialFunction=&BlendAngleCorrectedNormals}.Result
   False=@n18
  ) {defaultValue=true, parameterName=UseVertShader}
  VertexInterpolator(VS=@n16.Result Inverted)
 ]
}
